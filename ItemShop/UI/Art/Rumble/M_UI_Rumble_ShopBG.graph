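# UI material for the Rumble item shop background, drawing animated fiery orange-red shapes over a black-to-deep-blue radial gradient.
# Vertical gradients and panning blurred noise build a mask that blends confetti from MF_UI_Confetti with a scrolling movement texture, tinted by orange and radial red gradients.
# Named reroutes define aspect-corrected UVs, noise distortion and time-stepped, grid-snapped "PixelPush" coordinates.
# Custom HLSL Circle and MovingTiles nodes appear to be unused leftovers.
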Material /ItemShop/UI/Art/Rumble/M_UI_Rumble_ShopBG
graph {
 properties {bUsedWithStaticMesh=false, materialDomain=MD_UI, pixelDepthOffsetMode=PDOM_Legacy, usedShadingModels=Default Lit}
 nodes [
  Add()
  n1 = NamedRerouteUsage()
  n2 = AppendVector(A=ScalarParameter() {defaultValue=0.7793, parameterName=Radial-X-Tile}, B=ScalarParameter() {defaultValue=1, parameterName=Radial-Y-Tile})
  n3 = Constant2Vector() {g=1, r=0.5}
  n4 = Power(Base=Distance(A=Multiply(A=@n1, B=@n2), B=Multiply(A=@n2, B=@n3)), Exp=ScalarParameter() {defaultValue=1.514, parameterName=Distance Power})
  n5 = Subtract(A=ComponentMask(None=TextureCoordinate()) {g=true}, B=ScalarParameter() {defaultValue=1.7796, parameterName=Subtract})
  n6 = ScalarParameter() {defaultValue=2.6362, parameterName=Falloff}
  n7 = LinearInterpolate(B=Constant3Vector() {constant=<rgba 1 0.6749 0 0>}, Alpha=Saturate(None=Multiply(A=@n5, B=@n6)))
  n8 = NamedRerouteUsage()
  n9 = AppendVector(A=ScalarParameter() {parameterName=L1-Patter-X-Offset}, B=ScalarParameter() {defaultValue=0.172, parameterName=L1-Patter-Y-Offset})
  n10 = ScalarParameter() {defaultValue=50, parameterName=L1-Pattern-X}
  n11 = ScalarParameter() {defaultValue=0.75, parameterName=L1-Pattern-Y}
  n12 = ScalarParameter() {defaultValue=5.2803, parameterName=L1-Confetti-Local-Tile-X}
  n13 = ScalarParameter() {defaultValue=111.9648, parameterName=L1-Confetti-Local-Tile-Y}
  n14 = ScalarParameter() {defaultValue=0.6489, parameterName=L1-SmoothStep-Max}
  n15 = ScalarParameter() {defaultValue=0.5, parameterName=L1-SmoothStep-Min}
  n16 = ComponentMask(None=TextureCoordinate()) {g=true}
  n17 = Subtract(A=@n16, B=ScalarParameter() {defaultValue=0.34, parameterName=L1-Gradient-Mask})
  n18 = ScalarParameter() {defaultValue=2, parameterName=L1-Gradient-Falloff}
  n19 = Multiply(
   A=MaterialFunctionCall!(
    "UV(V2)"=Add(A=@n8, B=@n9)
    Global-Angle=ScalarParameter() {defaultValue=-1, parameterName=L1-Global Angle}
    "Pattern-Tiles(V2)"=AppendVector(A=@n10, B=@n11)
    RandomSeed=ScalarParameter() {defaultValue=12345, parameterName=L1-RandomSeed}
    Confetti-Speed-TimeFactor=ScalarParameter() {defaultValue=0.138, parameterName=L1-Confetti-Speed-TimeFactor}
    Confetti-Rotation-TimeFactor=ScalarParameter() {defaultValue=0.794, parameterName=L1-Confetti-Rotation-TimeFactor}
    Confetti-Local-Rotation-Range=ScalarParameter() {defaultValue=0.0075, parameterName=L1-Confetti-Local-Rotation-Range}
    "Confetti-Local-Tiles(V2)"=AppendVector(A=@n12, B=@n13)
    "SmoothStepRanges(V2)"=AppendVector(A=@n14, B=@n15)
   ) {materialFunction=/Game/UI/Frontend/ItemShop/Art/Cactus/MF_UI_Confetti}.Confetti-Mask
   B=Saturate(None=Multiply(A=@n17, B=@n18))
  )
  n20 = ScalarParameter() {defaultValue=1, parameterName=L1-Intensity}
  n21 = TextureCoordinate()
  n22 = Multiply(
   A=Subtract(A=ComponentMask(None=@n21) {g=true}, B=ScalarParameter() {defaultValue=0.5854, parameterName=Subtract2})
   B=ScalarParameter() {defaultValue=0.2774, parameterName=Falloff2}
  )
  n23 = Panner(
   Coordinate=Add(A=NamedRerouteUsage(), B=NamedRerouteUsage())
   Time=Multiply(A=Time(), B=ScalarParameter() {defaultValue=0.6612, parameterName=TopMaskScale})
  ) {speedX=0.015, speedY=0.104}
  n24 = MaterialFunctionCall(
   GradientStart=ScalarParameter() {defaultValue=-0.292, parameterName=Param_2, sliderMin=0.056}
   RotationAngle=ScalarParameter() {defaultValue=0.25, parameterName=rotation}
   GradientOffset=ScalarParameter() {defaultValue=-0.216, parameterName=TopMaskGend}
  ) {materialFunction=&MF_UI_LinearGradient}
  n25 = Add(
   A=Multiply(A=Saturate(None=@n22), B=ScalarParameter() {defaultValue=11.0695, parameterName=Intensity})
   B=LinearInterpolate(
    B=TextureSample(UVs=@n23) {samplerType=SAMPLERTYPE_Grayscale, texture=/Game/UI/Frontend/ItemShop/Art/Rumble/LowResBlurredNoise_Rumble}.RGB
    Alpha=OneMinus(None=@n24.Result)
   )
  )
  n26 = Panner(Coordinate=Add(A=NamedRerouteUsage(), B=NamedRerouteUsage()), Time=Multiply(A=Time(), B=ScalarParameter() {defaultValue=0.592})) {speedY=0.25}
  n27 = LinearInterpolate(
   A=Multiply(A=Constant3Vector() {constant=<rgba 1 0.3366 0.0015 1>}, B=Multiply(A=@n19, B=@n20))
   B=Multiply(A=@n25, B=TextureSample(UVs=@n26) {texture=/ItemShop/UI/Art/Cactus/Textures/T_UI_CactusMovement}.RGB)
  )
  n28 = AppendVector(
   A=ScalarParameter() {defaultValue=0.9863, parameterName=Radial-X-Tile2}
   B=ScalarParameter() {defaultValue=1.3969, parameterName=Radial-Y-Tile2}
  )
  n29 = Distance(A=Multiply(A=NamedRerouteUsage(), B=@n28), B=Multiply(A=@n28, B=Constant2Vector() {g=1, r=0.5}))
  n30 = ScalarParameter() {defaultValue=0.5829, parameterName=Distance Powerfire}
  n31 = LinearInterpolate(
   A=Constant3Vector() {constant=<rgba 1 0.1298 0 0>}
   B=Constant3Vector() {constant=<rgba 0.2604 0.0133 0 0>}
   Alpha=Saturate(None=Power(Base=@n29, Exp=@n30))
  )
  LinearInterpolate(
   A=LinearInterpolate(A=Constant3Vector(), B=Constant3Vector() {constant=<rgba 0.001 0 0.1823 0>}, Alpha=Saturate(None=@n4))
   B=LinearInterpolate(A=LinearInterpolate(B=@n7, Alpha=@n27), B=Multiply(A=@n27, B=@n31), Alpha=@n25)
  )
  MaterialExpressionConstant_6 = Constant() {r=0.3}
  n32 = Custom!(_st, _radius=@MaterialExpressionConstant_6) {
   code="float2 pos = float2(0.5, 0.5)-_st;\r\nreturn smoothstep(1.0-_radius,1.0-_radius+_radius*0.2,1.-dot(pos,pos)*3.14);"
   description=Circle
   outputType=CMOT_Float1
   inputs [
    {inputName=_st, input {expression=null}}
    {inputName=_radius, input {expression=&self:MaterialExpressionConstant_6}}
   ]
  }
  n33 = Subtract(B=@n32)
  MaterialFunctionCall(X=@n33, Y=@n33, Z=@n33, A=Constant() {r=1}) {materialFunction=&MakeFloat4}
  MaterialExpressionNamedRerouteUsage_13 = NamedRerouteUsage()
  MaterialExpressionTime_7 = Time()
  MaterialExpressionConstant_3 = Constant() {r=0.1}
  MaterialExpressionConstant_4 = Constant() {r=50}
  n34 = Custom!(
   _st=@MaterialExpressionNamedRerouteUsage_13
   u_time=@MaterialExpressionTime_7
   _speed=@MaterialExpressionConstant_3
   _zoom=@MaterialExpressionConstant_4
  ) {
   code="_st *= _zoom;\r\nfloat time = u_time*_speed;\r\nif( frac(time)>0.5 ){\r\n    if (frac( _st.y * 0.5) > 0.5){\r\n        _st.x += frac(time)*2.0;\r\n    } else {\r\n        _st.x -= frac(time)*2.0;\r\n    }\r\n} else {\r\n    if (frac( _st.x * 0.5) > 0.5){\r\n        _st.y += frac(time)*2.0;\r\n    } else {\r\n        _st.y -= frac(time)*2.0;\r\n    }\r\n}\r\nreturn frac(_st);"
   description=MovingTiles
   outputType=CMOT_Float2
   inputs [
    {inputName=_st, input {expression=&self:MaterialExpressionNamedRerouteUsage_13}}
    {inputName=u_time, input {expression=&self:MaterialExpressionTime_7}}
    {inputName=_speed, input {expression=&self:MaterialExpressionConstant_3}}
    {inputName=_zoom, input {expression=&self:MaterialExpressionConstant_4}}
   ]
  }
  n35 = MaterialFunctionCall(Float2=@n34) {materialFunction=&BreakOutFloat2Components}
  MaterialFunctionCall(X=@n35.R, Y=@n35.R, Z=Constant() {r=1}) {materialFunction=&MakeFloat3}
  Multiply(B=ScalarParameter() {defaultValue=0.7653, parameterName=DupeScaler})
  n36 = MaterialFunctionCall() {materialFunction=&GetUserInterfaceUV}
  n37 = MaterialFunctionCall(Float2=@n36.Pixel Size) {materialFunction=&BreakOutFloat2Components}
  n38 = AppendVector(A=Divide(A=@n37.R, B=Max(A=@n37.G)), B=Constant() {r=1})
  n39 = OneMinus(None=@n38)
  NamedRerouteDeclaration(None=Add(A=Multiply(A=@n39) {constB=0.5}, B=Multiply(A=@n38, B=@n36.Normalized UV))) {name="Squared UV - (x/y,y)", nodeColor=<rgba 0 0 0 1>}
  n40 = Panner(
   Coordinate=Multiply(A=TextureCoordinate(), B=ScalarParameter() {defaultValue=0.7653, parameterName=DupeScaler})
   Time=Multiply(A=Time(), B=ScalarParameter() {defaultValue=0.3191, parameterName=TimeFactor})
  ) {speedX=0.058, speedY=0.096}
  NamedRerouteDeclaration(None=Multiply(
    A=TextureSample(UVs=@n40) {samplerType=SAMPLERTYPE_Grayscale, texture=&LowResBlurredNoise}.R
    B=ScalarParameter() {defaultValue=0.5164, parameterName=DistortionScaler}
   )) {name=UV-Distortion, nodeColor=<rgba 0 0 0 1>}
  n41 = NamedRerouteUsage()
  n42 = NamedRerouteUsage()
  n43 = Time()
  n44 = ScalarParameter() {defaultValue=0.3191, parameterName=TimeFactor}
  n45 = Power(
   Base=Panner(Coordinate=Add(A=@n41, B=@n42), Time=Multiply(A=@n43, B=@n44)) {speedX=16, speedY=0.616}
   Exp=ScalarParameter() {defaultValue=0.5465, parameterName=ConfettiParam2}
  )
  n46 = NamedRerouteUsage()
  NamedRerouteDeclaration(None=TextureSample(UVs=Add(A=@n45, B=@n46)) {samplerType=SAMPLERTYPE_Grayscale, texture=&LowResBlurredNoise_Rumble}.RGB) {name=DistortionMask, nodeColor=<rgba 0 0 0 1>}
  n47 = ScalarParameter() {parameterName=8}
  n48 = Multiply(A=Multiply(A=Time(), B=ScalarParameter() {defaultValue=2.4662, parameterName=Time Factor6}), B=@n47)
  n49 = AppendVector(A=ScalarParameter() {defaultValue=-0.1599, parameterName=0}, B=Divide(A=Floor(None=@n48), B=@n47))
  n50 = NamedRerouteUsage()
  n51 = ScalarParameter() {defaultValue=1.1439, parameterName=TextureScaler}
  n52 = ScalarParameter() {defaultValue=35, parameterName=grid}
  n53 = Multiply(A=Add(A=Panner(Coordinate=@n49) {speedY=0.05}, B=Multiply(A=@n50, B=@n51)), B=@n52)
  NamedRerouteDeclaration(None=Divide(A=Floor(None=@n53), B=@n52)) {name=PixelPush, nodeColor=<rgba 0 0 0 1>}
  NamedRerouteUsage()
  OneMinus(None=@n16)
  Power(Base=@n40, Exp=ScalarParameter() {defaultValue=0.5465, parameterName=ConfettiParam2})
 ]
}
